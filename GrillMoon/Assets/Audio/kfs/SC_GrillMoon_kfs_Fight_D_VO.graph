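# Sound cue for the GrillMoon Fight D voice-over, apparently one stem of a sequence.
# A single wave player feeds the Fight D VO wave straight to the output, with no randomization or mixing.
# It uses the shared GrillMoon character-stem attenuation with a very large max distance, routes to the Fort_Events sound class and slightly boosts volume.
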SoundCue /GrillMoon/Assets/Audio/kfs/SC_GrillMoon_kfs_Fight_D_VO
root=WavePlayer_1
nodes [{id=WavePlayer_1, properties {soundWaveAssetPtr=/GrillMoon/Assets/Audio/wav/GrillMoon_Seq_Fight_D_VO}}]
properties {
 attenuationSettings=/GrillMoon_Audio/Assets/Settings/ATT_GrillMoon_kfs_CharacterStems
 maxDistance=2097152
 soundClassObject=&Fort_Events
 volumeMultiplier=1.2
}
