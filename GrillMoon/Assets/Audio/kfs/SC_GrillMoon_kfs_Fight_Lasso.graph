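# Sound cue for the lasso player audio in a GrillMoon fight sequence, likely a keyframed cinematic stem judging by its kfs naming.
# A single wave player feeds the output with the LassoPlayer sequence wave, without randomization or mixing.
# It uses the GrillMoon character stems attenuation and the Fort_Events sound class.
# A partial pre-effect send feeds a music-volume sidechain bus, presumably to duck music while it plays.
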SoundCue /GrillMoon/Assets/Audio/kfs/SC_GrillMoon_kfs_Fight_Lasso
root=WavePlayer_1
nodes [{id=WavePlayer_1, properties {soundWaveAssetPtr=/GrillMoon/Assets/Audio/wav/GrillMoon_Seq_Fight_LassoPlayer}}]
properties {
 attenuationSettings=/GrillMoon_Audio/Assets/Settings/ATT_GrillMoon_kfs_CharacterStems
 maxDistance=2097152
 soundClassObject=&Fort_Events
 volumeMultiplier=1
 preEffectBusSends [{audioBus=&AB_Sidechain_FortMusic_Vol, sendLevel=0.3, soundSourceBus=null}]
}
