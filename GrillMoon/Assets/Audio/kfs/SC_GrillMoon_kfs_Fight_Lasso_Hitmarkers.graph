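# Sound cue that plays the lasso hit marker audio for the GrillMoon fight sequence, likely timed to a cinematic given its kfs and Seq naming.
# A single wave player feeds the root directly, with no randomization or modulation.
# It routes through the Fort_Events sound class at slightly reduced volume, with a very large max distance.
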SoundCue /GrillMoon/Assets/Audio/kfs/SC_GrillMoon_kfs_Fight_Lasso_Hitmarkers
root=WavePlayer_1
nodes [{id=WavePlayer_1, properties {soundWaveAssetPtr=/GrillMoon/Assets/Audio/wav/GrillMoon_Seq_Fight_Lasso_HitMarkers}}]
properties {maxDistance=2097152, soundClassObject=&Fort_Events, volumeMultiplier=0.9}
