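# Sound cue for the GrillMoon Fight sweetener, apparently a 5.1 layer for a fight sequence.
# A single wave player feeds the Fight_Sweetener_5_1 wave straight to the output at a slightly boosted volume, routed through the Fort_Events sound class.
# Attenuation overrides turn off distance falloff and reverb send and use 2D surround for the non-spatialized radius mode.
# Submix sends are disabled.
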SoundCue /GrillMoon/Assets/Audio/kfs/SC_GrillMoon_kfs_Fight_Sweetener_5_1
root=WavePlayer_0
nodes [{id=WavePlayer_0, properties {soundWaveAssetPtr=/GrillMoon/Assets/Audio/wav/GrillMoon_Seq_Fight_Sweetener_5_1}}]
properties {
 bEnableSubmixSends=false
 bOverrideAttenuation=true
 maxDistance=2097152
 soundClassObject=&Fort_Events
 volumeMultiplier=1.1
 attenuationOverrides {bAttenuate=false, bEnableReverbSend=false, nonSpatializedRadiusMode=Surround2D, stereoSpread=5000}
}
